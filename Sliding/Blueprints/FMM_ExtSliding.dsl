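; FMM_ExtSliding is the Blueprint for Athena's sliding movement mode extension, implementing event hooks for slide camera and landing effects.
; EventSetCameraBlendIn casts the supplied modifier to CameraModifier_Athena_Sliding and, despite the event's name, calls BlendOut on it.
; Two toggle events set the hidden-in-game state of the walk-landing effect, one via a PlayerPawn_Generic cast and the other on FortPlayerPawnAthena's particle component.
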(event AthenaSliding|EventSetCameraBlendIn
  (params (SlideCameraModifier :type Camera%20Modifier%20Object%20Reference))
  (bind _ascamera_modifier_athena_sliding (Utilities|Casting|CastToCameraModifier_Athena_Sliding SlideCameraModifier))
  (Class|CameraModifierAthenaSliding|BlendOut _ascamera_modifier_athena_sliding))

(event AthenaSliding|EventToggleLandFX
  (params (FortPlayerPawn :type Fort%20Player%20Pawn%20Object%20Reference) (bHiddenInGame :type Boolean))
  (bind _asplayer_pawn_generic (Utilities|Casting|CastToPlayerPawn_Generic FortPlayerPawn))
  (Development|SetHiddeninGame (Variables|/Game/Abilities/Player/Pawns/PlayerPawn_Generic.PlayerPawn_Generic_C|GetEffect_Player_Walk_Land _asplayer_pawn_generic) bHiddenInGame false))

(event AthenaSliding|EventToggleLandFXAthena
  (params (FortPlayerPawnAthena :type Fort%20Player%20Pawn%20Athena%20Object%20Reference) (bHiddenInGame :type Boolean))
  (Development|SetHiddeninGame (Variables|/Script/FortniteGame.FortPlayerPawnAthena|GetPSC_PlayerWalkLand FortPlayerPawnAthena) bHiddenInGame false))
